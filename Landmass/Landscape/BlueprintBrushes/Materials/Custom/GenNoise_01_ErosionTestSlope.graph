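# Unlit Landmass landscape brush material, apparently a test for erosion on slopes, that modifies and repacks an RG8 height render target.
# It adds an elevation offset and rerouted term to unpacked height, then lerps from LipHeight, clamps and packs.
# Named reroutes build normalized UVs, a border mask and a synthetic channel/plateau/ramp shape.
# Custom HLSL fractal noise drives the elevation lerp, but both ends read Elevation Min, so it has no effect.
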Material /Landmass/Landscape/BlueprintBrushes/Materials/Custom/GenNoise_01_ErosionTestSlope
graph {
 properties {bAllowNegativeEmissiveColor=true, pixelDepthOffsetMode=PDOM_Legacy, shadingModel=MSM_Unlit, usedShadingModels=Unlit}
 nodes [
  n1 = ScalarParameter() {parameterName=LipHeight}
  n2 = TextureSampleParameter2D() {parameterName=HeightRT, texture=/Engine/EngineResources/Black_Low.Black}
  n3 = VectorParameter() {desc="R=Elevation Max, G=Elevation Min, B=Pre-Offset, A=Invert", parameterName=ElevationParams}
  n4 = Divide(A=AppendVector(
    A=CollectionParameter() {collection=/Landmass/Landscape/BlueprintBrushes/MPC/MPC_Landscape, parameterName=WorldSizeX}
    B=CollectionParameter() {collection=&MPC_Landscape, parameterName=WorldSizeY}
   )) {constB=1000}
  n5 = Multiply(A=TextureCoordinate(), B=@n4)
  n6 = Divide(A=ComponentMask(None=VectorParameter() {parameterName=Position}.RGB) {g=true, r=true}) {constB=1000}
  MaterialExpressionVectorParameter_0 = VectorParameter() {desc="R=Tiling, G=Seed, B=Octaves, A=Persistence", parameterName=NoiseSeedParams, defaultValue=<rgba 2 0 2 0.5>}
  MaterialExpressionAppendVector_0 = AppendVector(A=Multiply(A=Subtract(A=@n5, B=@n6), B=@MaterialExpressionVectorParameter_0.R), B=@MaterialExpressionVectorParameter_0.G)
  MaterialExpressionConstant_2 = Constant() {desc="3 = Gradient, 5 = Voronoi", r=3}
  MaterialExpressionConstant_3 = Constant()
  MaterialExpressionMultiply_9 = Multiply(A=@n4, B=@MaterialExpressionVectorParameter_0.R)
  n7 = Custom!(
   Position=@MaterialExpressionAppendVector_0
   Function=@MaterialExpressionConstant_2
   bTurbulence=@MaterialExpressionConstant_3
   Octaves=@MaterialExpressionVectorParameter_0.B
   Persistence=@MaterialExpressionVectorParameter_0.B
   RepeatSize=@MaterialExpressionMultiply_9
  ) {
   code="\r\nfloat Out = 0;\r\nfloat OutScale = 1;\r\n\r\n[loop]\r\nfor(int i = 0; i < Octaves; ++i)\r\n{\r\n\r\n\r\n\tif(bTurbulence)\r\n\t{\r\n\t\tOut += abs(Noise3D_Multiplexer(Function, Position, 4, 1, RepeatSize)) * OutScale;\r\n\t}\r\n\telse\r\n\t{\r\n\t\tOut += Noise3D_Multiplexer(Function, Position, 4, 1, RepeatSize) * OutScale;\r\n\t}\r\n\r\n\tPosition *= 2;\r\n\tRepeatSize *= 2;\r\n\tOutScale *= Persistence;\r\n\r\n}\r\n\r\nif(!bTurbulence)\r\n{\r\n\t// bring -1..1 to 0..1 range\r\n\tOut = Out * 0.5f + 0.5f;\r\n}\r\n\r\n// Out is in 0..1 range\r\nreturn lerp(0, 1, Out);"
   description=CustomNoise
   outputType=CMOT_Float1
   inputs [
    {
     inputName=Position
     input {expression=&self:MaterialExpressionAppendVector_0}
    }
    {
     inputName=Function
     input {expression=&self:MaterialExpressionConstant_2}
    }
    {
     inputName=bTurbulence
     input {expression=&self:MaterialExpressionConstant_3}
    }
    {
     inputName=Octaves
     input {
      expression=&self:MaterialExpressionVectorParameter_0
      mask=1
      maskB=1
      outputIndex=3
     }
    }
    {
     inputName=Persistence
     input {
      expression=&self:MaterialExpressionVectorParameter_0
      mask=1
      maskA=1
      outputIndex=4
     }
    }
    {
     inputName=RepeatSize
     input {expression=&self:MaterialExpressionMultiply_9}
    }
   ]
  }
  n8 = OneMinus(None=@n7)
  n9 = Max(A=Subtract(A=LinearInterpolate(A=@n7, B=@n8, Alpha=@n3.A), B=@n3.B)) {constB=0}
  n10 = Add(
   A=MaterialFunctionCall!(RG=ComponentMask(None=@n2.RGB) {g=true, r=true}) {materialFunction=/Landmass/Landscape/OldPrototype_BP/Mats/MF/UnpackRG8Height}.Result
   B=Add(A=LinearInterpolate(A=@n3.G, B=@n3.G, Alpha=@n9), B=NamedRerouteUsage())
  )
  n11 = NamedRerouteUsage()
  MaterialFunctionCall!(Height Int -32k to 32k=Clamp(None=LinearInterpolate(A=@n1, B=@n10, Alpha=@n11)) {maxDefault=32767, minDefault=-32768}) {materialFunction=/Landmass/Landscape/OldPrototype_BP/Mats/MF/PackRG8Height}
  n12 = TextureCoordinate()
  n13 = AppendVector(
   A=CollectionParameter() {collection=&MPC_Landscape, parameterName=LSQuadsX}
   B=CollectionParameter() {collection=&MPC_Landscape, parameterName=LSQuadsY}
  )
  n14 = Divide(
   A=Add(A=Add(A=@n13), B=ScalarParameter() {parameterName=Offset to test Lip})
   B=AppendVector(
    A=CollectionParameter() {collection=&MPC_Landscape, parameterName=RTResX}
    B=CollectionParameter() {collection=&MPC_Landscape, parameterName=RTResY}
   )
  )
  NamedRerouteDeclaration(None=Saturate(None=Divide(A=@n12, B=@n14))) {name=Normalized UV, nodeColor=<rgba 0.0312 0 1 1>}
  n15 = NamedRerouteUsage()
  n16 = AppendVector(
   A=CollectionParameter() {collection=&MPC_Landscape, parameterName=LSQuadsX}
   B=CollectionParameter() {collection=&MPC_Landscape, parameterName=LSQuadsY}
  )
  n17 = Divide(
   A=Subtract(A=Multiply(A=@n15, B=@n16) {desc=Vertex Indices}, B=ScalarParameter() {defaultValue=8, parameterName=Edge Offset Verts})
   B=ScalarParameter() {defaultValue=16, parameterName=Edge Mask Width}
  )
  n18 = OneMinus(None=NamedRerouteUsage())
  n19 = AppendVector(
   A=CollectionParameter() {collection=&MPC_Landscape, parameterName=LSQuadsX}
   B=CollectionParameter() {collection=&MPC_Landscape, parameterName=LSQuadsY}
  )
  n20 = Divide(
   A=Subtract(A=Multiply(A=@n18, B=@n19) {desc=Vertex Indices}, B=ScalarParameter() {defaultValue=8, parameterName=Edge Offset Verts})
   B=ScalarParameter() {defaultValue=16, parameterName=Edge Mask Width}
  )
  n21 = MaterialFunctionCall(Float2=Min(A=Saturate(None=@n17), B=Saturate(None=@n20))) {materialFunction=&BreakOutFloat2Components}
  NamedRerouteDeclaration(None=Min(A=@n21.R, B=@n21.R)) {name=Edge Mask, nodeColor=<rgba 0 0.2734 1 1>}
  n22 = TextureCoordinate()
  n23 = ComponentMask(None=@n22) {r=true}
  n24 = Subtract(A=Abs(None=Subtract(A=@n23) {constB=0.5})) {constB=0.5}
  n25 = ScalarParameter() {parameterName=Channel}
  n26 = OneMinus(None=ComponentMask(None=@n22) {g=true})
  n27 = Ceil(None=Subtract(A=@n26) {constB=0.5})
  n28 = Multiply(A=Subtract(A=Saturate(None=@n27)) {constB=0.5}, B=ScalarParameter() {parameterName=Plateau})
  n29 = Multiply(A=Subtract(A=@n26) {constB=0.5}, B=ScalarParameter() {parameterName=Ramp})
  NamedRerouteDeclaration(None=Add(A=Multiply(A=@n24, B=@n25), B=Add(A=@n28, B=@n29))) {name=Erosion Channel, nodeColor=<rgba 0.9688 0 1 1>}
  Noise!(World Position=@MaterialExpressionAppendVector_0, FilterWidth) {bTurbulence=false, levelScale=3, levels=16, noiseFunction=NOISEFUNCTION_VoronoiALU, outputMin=0, quality=5, repeatSize=2}
  ScalarParameter() {parameterName=Seed}
  ScalarParameter() {defaultValue=2, parameterName=Tiling}
  ScalarParameter() {defaultValue=5, parameterName=NoiseType}
  ScalarParameter() {defaultValue=0.5, parameterName=Persistence}
  ScalarParameter() {defaultValue=0.3, parameterName=Elevation Pre-Offset}
  ScalarParameter() {defaultValue=3, parameterName=Octaves}
  ScalarParameter() {defaultValue=1, parameterName=Invert}
 ]
}
